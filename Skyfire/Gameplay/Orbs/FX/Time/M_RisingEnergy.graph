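# Additive, unlit, two-sided material for Niagara mesh particles in the Time orb effects, drawing rising, noisy energy.
# Two panned smoke noises, tiled and sped by dynamic parameters, blend to drive dark/light/center color, opacity and optional normal-based vertex displacement with Niagara-driven pinch.
# UV gradients, glow, vertex color, view-angle, depth and optional near-camera fades shape opacity.
# It also builds camera-facing Z rotation, HDR color with mobile fallbacks, and quality-switched distortion.
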Material /Skyfire/Gameplay/Orbs/FX/Time/M_RisingEnergy
graph {
 properties {
  bUsedWithNiagaraMeshParticles=true
  blendMode=BLEND_Additive
  pixelDepthOffsetMode=PDOM_Legacy
  shadingModel=MSM_Unlit
  twoSided=true
  usedShadingModels=Unlit
 }
 nodes [
  ActorPositionWS()
  n1 = DynamicParameter() {paramNames [YTile,PanSpeed,Param3,Param4]}
  n2 = AppendVector(A=ScalarParameter() {defaultValue=1, parameterName=X_Tile}, B=@n1.YTile)
  n3 = Multiply(A=Time(), B=@n1.PanSpeed)
  n4 = Panner(
   Coordinate=Multiply(A=@n2, B=TextureCoordinate() {uTiling=3, vTiling=0.5})
   Time=@n3
   Speed=AppendVector(A=ScalarParameter() {defaultValue=0.2, parameterName=NoiseA_SpeedX}, B=ScalarParameter() {defaultValue=1, parameterName=NoiseA_SpeedY})
  ) {speedX=0.2, speedY=1}
  n5 = TextureSampleParameter2D(UVs=@n4) {parameterName=WPO_NoiseA, texture=&T_Inky_Smoke_Tile_Chromatic}
  n6 = Panner(
   Coordinate=Multiply(A=@n2, B=TextureCoordinate() {uTiling=3, vTiling=0.2})
   Time=@n3
   Speed=AppendVector(A=ScalarParameter() {defaultValue=-0.25, parameterName=NoiseB_SpeedX}, B=ScalarParameter() {defaultValue=0.8, parameterName=NoiseB_SpeedY})
  ) {speedX=-0.25, speedY=0.8}
  n7 = TextureSampleParameter2D(UVs=@n6) {parameterName=WPO_NoiseB, texture=&T_Inky_Smoke_Tile_Chromatic}
  n8 = TextureSampleParameter2D(UVs=@n4) {parameterName=NoiseA, texture=&T_Inky_Smoke_Tile_Chromatic}
  n9 = TextureSampleParameter2D(UVs=@n6) {parameterName=NoiseB, texture=&T_Inky_Smoke_Tile_Chromatic}
  n10 = Multiply(A=Power(Base=LinearInterpolate(A=@n8.RGB, B=@n9.RGB), Exp=ScalarParameter() {defaultValue=1, parameterName=Noise_Exponent})) {constB=1.5}
  n11 = Multiply(A=VertexNormalWS(), B=StaticSwitchParameter(True=LinearInterpolate(A=@n5.RGB, B=@n7.RGB), False=@n10) {parameterName=UseSpecificWPOtexture})
  n12 = ScalarParameter() {parameterName=WPO}
  n13 = DynamicParameter() {desc=Niagara index 1, parameterIndex=1, defaultValue=<rgba 0 1 1 1>, paramNames [WPOPinch,GlowMult,Param3,Param4]}
  n14 = Subtract(A=WorldPosition().XYZ, B=ParticlePositionWS())
  n15 = Power(Base=ComponentMask(None=TextureCoordinate()) {g=true}) {constExponent=3}
  Add(
   A=StaticSwitchParameter(True=Multiply(A=@n11, B=@n12) {constB=7}, False=Constant()) {parameterName=UseWPO}
   B=Multiply(A=@n13.WPOPinch, B=Multiply(A=@n14, B=@n15))
  )
  n16 = CameraPositionWS()
  n17 = ActorPositionWS()
  n18 = Normalize(VectorInput=Multiply(A=Constant3Vector() {constant=<rgba 1 1 0 0>}, B=Subtract(A=@n16, B=@n17)))
  n19 = Constant3Vector() {constant=<rgba 0 1 0 0>}
  n20 = Divide(A=MaterialFunctionCall!(Normalized V1=@n18, Normalized V2=Normalize(VectorInput=@n19)) {materialFunction=/Engine/Functions/Engine_MaterialFunctions02/Math/AngleBetweenVectors}.Angle In Degrees) {constB=360}
  n21 = Normalize(VectorInput=Constant3Vector() {constant=<rgba 1 0 0 0>})
  n22 = Multiply(A=Sign(None=DotProduct(A=@n18, B=@n21))) {constB=-1}
  LinearInterpolate(
   B=RotateAboutAxis(
    NormalizedRotationAxis=Constant3Vector() {constant=<rgba 0 0 1 0>}
    RotationAngle=Multiply(A=@n20, B=@n22)
    PivotPoint=@n17
    Position=WorldPosition().XYZ
   )
   Alpha=ScalarParameter() {defaultValue=1, parameterName=FaceCamera}
  )
  n23 = MaterialFunctionCall(Float2=TextureCoordinate()) {materialFunction=&BreakOutFloat2Components}
  n24 = OneMinus(None=@n23.R)
  n25 = Multiply(
   A=Power(Base=Multiply(A=@n23.R, B=@n24), Exp=ScalarParameter() {defaultValue=1, parameterName=Directional Mask Power}) {constExponent=1.5}
   B=ScalarParameter() {defaultValue=1, parameterName=Directional Mask Hardness}
  )
  n26 = TextureCoordinate()
  n27 = SphereMask(
   A=Power(Base=MaterialFunctionCall(Float2=@n26) {materialFunction=&BreakOutFloat2Components}.G) {constExponent=1.5}
   B=Constant() {r=0.5}
   Radius=LinearInterpolate(
    A=ScalarParameter() {defaultValue=0.5, group=DetailedGradientMask, parameterName=LowDetailedRange}
    B=ScalarParameter() {defaultValue=0.5, group=DetailedGradientMask, parameterName=HighDetailedRange}
    Alpha=@n5.R
   ) {constB=0.5}
  ) {attenuationRadius=0.5, hardnessPercent=0}
  n28 = ScalarParameter() {defaultValue=2.2, group=DetailedGradientMask, parameterName=PowerOfDetailedRange}
  n29 = DynamicParameter() {desc=Niagara index 1, parameterIndex=1, defaultValue=<rgba 0 1 1 1>, paramNames [WPOPinch,GlowMult,Param3,Param4]}
  n30 = ComponentMask(None=TextureCoordinate()) {g=true}
  n31 = ScalarParameter() {defaultValue=6, parameterName=GlowPower}
  n32 = Multiply(
   A=@n29.GlowMult
   B=Multiply(A=ScalarParameter() {defaultValue=0.5, parameterName=GlowBrightness}, B=Power(Base=@n30, Exp=@n31) {constExponent=6}) {constB=0.5}
  )
  n33 = Multiply(
   A=Power(Base=@n30, Exp=ScalarParameter() {defaultValue=1, parameterName=NoiseFadePower})
   B=StaticSwitchParameter(True=@n10, False=Add(A=@n32, B=@n10)) {parameterName=Glow}
  )
  n34 = Multiply(A=ParticleColor().A, B=Add(A=@n32, B=@n33))
  n35 = VertexColor()
  n36 = Power(Base=@n35.R)
  n37 = Abs(None=DotProduct(A=CameraVectorWS(), B=Constant3Vector() {constant=<rgba 0 0 1 0>}))
  n38 = ScalarParameter() {defaultValue=1.25, parameterName=DirectionalFadePowerLow}
  n39 = ScalarParameter() {defaultValue=4, parameterName=DirectionalFadePower}
  n40 = VertexColor()
  n41 = Multiply(A=Power(Base=OneMinus(None=@n37), Exp=LinearInterpolate(A=@n38, B=@n39, Alpha=@n40.G) {constA=1.25}) {constExponent=4}) {constB=4}
  n42 = DepthFade(FadeDistance=ScalarParameter() {defaultValue=50, parameterName=DepthFadeDist})
  n43 = Constant() {r=1}
  n44 = Multiply(A=Multiply(A=Multiply(A=@n34, B=@n36), B=Clamp(None=@n41)), B=FeatureLevelSwitch(Default=@n42, ES2=@n43, ES3_1=@n43, SM4_REMOVED=@n42))
  n45 = StaticSwitchParameter(True=Multiply(A=Power(Base=@n27, Exp=@n28) {constExponent=2.2}, B=@n44), False=@n44) {group=DetailedGradientMask, parameterName=UseDetailedGradientMask}
  n46 = StaticSwitchParameter(True=Saturate(None=Multiply(A=@n25, B=@n45)), False=@n45) {group=DetailedGradientMask, parameterName=UseDIrectionalGradientMask}
  n47 = ScalarParameter() {parameterName=CameraFadeDistance}
  n48 = StaticSwitchParameter(
   True=Multiply(A=@n46, B=MaterialFunctionCall(Distance=@n47) {materialFunction=&Near_Camera_Fade}.Result)
   False=@n46
  ) {group=DetailedGradientMask, parameterName=UseCameraNearFade}
  n49 = StaticSwitchParameter(True=Multiply(A=ParticleColor().A, B=@n48), False=@n48) {parameterName=UseParticleAlpha}
  LinearInterpolate(A=@n49, B=Multiply(A=@n49, B=VertexColor()), Alpha=ScalarParameter() {group=Opacity, parameterName=VertexOpacity})
  MaterialFunctionCall() {materialFunction=&DebugTimeSine}
  n50 = TextureCoordinate()
  n51 = AppendVector(
   A=Multiply(A=ComponentMask(None=@n50) {r=true}, B=ScalarParameter() {defaultValue=1, group=ChromaticNoise, parameterName=Noise-TilingX})
   B=Multiply(A=ComponentMask(None=@n50) {g=true}, B=ScalarParameter() {group=ChromaticNoise, parameterName=Noise-TilingY})
  )
  n52 = AppendVector(
   A=ScalarParameter() {group=ChromaticNoise, parameterName=NoisePanning-X}
   B=ScalarParameter() {defaultValue=0.2, group=ChromaticNoise, parameterName=NoisePanning-Y}
  )
  n53 = Desaturation(
   None=TextureSampleParameter2D(UVs=Panner(Coordinate=@n51, Speed=@n52)) {group=ChromaticNoise, parameterName=NoiseTexture, texture=/Game/Global/Textures/Default/Noise/T_Noise_Cloud}.RGB
   Fraction=ScalarParameter() {defaultValue=-0.5, group=ChromaticNoise, parameterName=Noise-Desat_Frac}
  )
  n54 = VectorParameter() {parameterName=DarkColor, defaultValue=<rgba 0.1393 0.0863 0.2344 0>}
  n55 = VectorParameter() {parameterName=LightColor, defaultValue=<rgba 1 1 1 0>}
  n56 = Multiply(
   A=LinearInterpolate(
    A=LinearInterpolate(A=@n54.RGB, B=@n55.RGB, Alpha=@n33)
    B=VectorParameter() {parameterName=CenterColor, defaultValue=<rgba 2 2 2 0>}.RGB
    Alpha=OneMinus(None=@n35.R)
   )
   B=ParticleColor().RGB
  )
  n57 = Clamp(None=Multiply(A=@n36, B=@n56))
  n58 = MaterialFunctionCall(LDR Input=@n57, HDR Tint=Multiply(A=@n56, B=ScalarParameter() {defaultValue=1.5, parameterName=HDRMult})) {materialFunction=&DeriveHDRfromLDR}
  n59 = FeatureLevelSwitch(Default=@n58.RemappedRange, ES2=@n57, ES3_1=@n57, SM4_REMOVED=@n58.RemappedRange)
  MaterialFunctionCall(Emissive=StaticSwitchParameter(True=Multiply(A=@n59, B=@n53), False=@n59) {group=ChromaticNoise, parameterName=UseChromaticNoise}) {materialFunction=&MF_FX_EmissiveAdjustments}
  n60 = Power(Base=DotProduct(A=CameraVectorWS(), B=VertexNormalWS()), Exp=ScalarParameter() {defaultValue=5, parameterName=Fresnel Size})
  StaticSwitchParameter(True=LinearInterpolate(A=@n44, B=Multiply(A=@n60, B=@n44), Alpha=TwoSidedSign()), False=@n44) {parameterName=Fresnel}
  n61 = ScalarParameter() {defaultValue=1, parameterName=Refraction}
  n62 = ScalarParameter() {parameterName=CameraFadeDistance}
  n63 = StaticSwitchParameter(
   True=LinearInterpolate(B=@n61, Alpha=MaterialFunctionCall(Distance=@n62) {materialFunction=&Near_Camera_Fade}.Result) {constA=1}
   False=@n61
  ) {group=DetailedGradientMask, parameterName=UseCameraNearFade}
  n64 = Desaturation(None=@n44, Fraction=Constant() {r=1})
  StaticSwitchParameter(True=QualitySwitch(Default=LinearInterpolate(B=@n63, Alpha=@n64) {constA=1, constB=1.2}, Low=Constant() {r=1}), False=Constant() {r=1}) {defaultValue=true, parameterName=Use_Distortion}
  VertexNormalWS()
 ]
}
