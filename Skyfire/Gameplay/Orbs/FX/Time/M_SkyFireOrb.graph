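# M_SkyFireOrb is a Default Lit material for the Skyfire time orb effect, flagged for use on Niagara mesh particles.
# It contains a Particle Color node and a Base Color vector parameter.
# An Options vector parameter packs Metallic, Specular and Roughness into its red, green and blue channels.
# No connections between these nodes or to material outputs are shown.
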Material /Skyfire/Gameplay/Orbs/FX/Time/M_SkyFireOrb
graph {
 properties {bUsedWithNiagaraMeshParticles=true, pixelDepthOffsetMode=PDOM_Legacy, usedShadingModels=Default Lit}
 nodes [
  ParticleColor()
  VectorParameter() {parameterName=Options, channelNames {a="", b=Roughness, g=Specular, r=Metallic}}
  VectorParameter() {parameterName=Base Color}
 ]
}
