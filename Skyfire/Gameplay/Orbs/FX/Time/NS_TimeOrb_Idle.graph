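# Niagara system for the idle look of the Time Orb, built from three distance-culled emitters.
# Icon_And_Light holds one looping particle with two camera-offset sprites and a blue light pulsed by a sine.
# Fake_Glow spawns soft sprites that fade and shrink.
# Squares_and_diamonds emits rising mesh and sprite particles, using a random VisibilityTag to switch color, velocity, fade curve and curl noise between two variants.
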NiagaraSystem /Skyfire/Gameplay/Orbs/FX/Time/NS_TimeOrb_Idle
dependencies {
 usedDataInterfaces [&NiagaraDataInterfaceCurve]
 usedDynamicInputs []
 usedModules [
  &SystemState
  &self:SystemSpawnScript.SystemScriptSource.SystemScriptGraph.NiagaraNodeAssignment_0.SetVariables_#1
 ]
}
emitters {
 Fake_Glow {
  emitterUpdateScript [
   &EmitterState(
    Enable Distance Culling=true
    Max Distance Response=::SleepAndClearParticles
    MaxDistance=15000
    Module.bApplySpawnCountScaleByDistance=true
    Module.bUseMaxDistance=true
    Module.bUseMinDistance=false
    Scalability Mode=::NewEnumerator1
    Scale Spawn Count=true
    Spawn Count Scale By Distance=DataInterface() {propertyValues {Curve {keys [<at 0 1>,<at 0.75 1>,<at 1 0.25>]}}}
   )
   &SpawnRate(Use Spawn Probability=false)
  ]
  particleSpawnScript [
   &InitializeParticle(
    Color=&HueShiftLinearColor(Hue=0, Saturation=2, Value=1, Linear Color=<rgba 0.4 0.5998 0.8 15>)
    Lifetime Max=3
    Lifetime Min=2
    Lifetime Mode=::NewEnumerator1
    Mass Mode=::NewEnumerator3
    Sprite Rotation Mode=::NewEnumerator2
    Sprite Size Mode=::NewEnumerator3
    Uniform Sprite Size=100
    UsePositionOffset=false
    Write Material Random=true
   )
   &SphereLocation(
    MaskSpawnGroup=false
    Override Seed=false
    Sphere Radius=2
    Surface Only=false
    UseNonUniformScale=false
    UseOffset=false
    UseSphereOrientation=false
   )
  ]
  particleUpdateScript [
   &ParticleState()
   &ScaleColor(
    Scale Alpha=&FloatFromCurve(FloatCurve=DataInterface() {propertyValues {Curve {keys [
         {interpMode=RCIM_Cubic, tangentMode=RCTM_User, time=0, value=0}
         {interpMode=RCIM_Cubic, tangentMode=RCTM_User, time=0.5, value=1}
         {tangentMode=RCTM_User, time=1, value=0}
        ]}}})
    ScaleA=true
    ScaleRGB=true
   )
   &ScaleSpriteSize(Scale Factor=&Vector2DFromFloat(Value=&Multiply_Float(
      A=&FloatFromCurve(FloatCurve=DataInterface() {propertyValues {Curve {keys [
           {arriveTangentWeight=0.5027, interpMode=RCIM_Cubic, leaveTangentWeight=0.4835, time=0, value=1}
           {arriveTangentWeight=0.5027, interpMode=RCIM_Cubic, time=1, value=0.25}
          ]}}})
      B=1
     )))
  ]
  rendererClasses [&NiagaraSpriteRendererProperties]
  renderers [{rendererClass=&NiagaraSpriteRendererProperties}]
 }
 Icon_And_Light {
  emitterUpdateScript [
   &EmitterState(
    Enable Distance Culling=true
    Enable Visibility Culling=true
    Life Cycle Mode=::NewEnumerator1
    Loop Duration=1
    Max Distance Response=::SleepAndClearParticles
    Module.bUseMaxDistance=true
    Module.bUseMinDistance=false
    Reset Age On Awaken=true
    Scalability Mode=::NewEnumerator1
    UseLoopDelay=false
    Visibility Cull Delay=0.1
    Visibility Cull Response:ENiagaraExecutionStateManagement=::SleepAndClearParticles
   )
   &SpawnBurst_Instantaneous(Age:NiagaraFloat=$Emitter.Age, Spawn Count=1, Use Spawn Probability=false)
  ]
  particleSpawnScript [
   &InitializeParticle(
    Color=<rgba 3.4 3.79 4 0.6>
    Lifetime=5
    Sprite Size Mode=::NewEnumerator3
    Uniform Sprite Size=45
    UsePositionOffset=false
    Write Material Random=true
   ) {module=InitializeParticle001}
   /SetVariables(
    Particles.BG Color:LinearColor=<rgba 0 0 0 0.5>
    Particles.BG Sprite Size:Vector2f=&Multiply_Vector2(A=$Particles.SpriteSize, B=&Vector2DFromFloat(Value=1))
   ) {isSetParameters=true}
   &CameraOffset(Camera Offset Amount=45)
  ]
  particleUpdateScript [
   &ParticleState(
    Kill Particles When Lifetime Has Elapsed=false
    Let Infinitely Lived Particles Die When Emitter Deactivates=true
    Loop Particles Lifetime=true
   )
   &FN_Light_Attributes(
    Color=<rgba 0.3 0.6497 1 1>
    Create Color Attribute=true
    Create Exponent Attribute=true
    Create Radius Attribute=true
    Exponent=3
    Radius=150
    Scale Color=&Lerp_Float(A=4, B=6, Alpha=&Sine(Normalized Angle=$Emitter.Age, Period=3, Scale=0.5, Bias=0.5))
   )
  ]
  rendererClasses [&NiagaraLightRendererProperties, &NiagaraSpriteRendererProperties]
  renderers [
   {rendererClass=&NiagaraLightRendererProperties, rendererIndex=0}
   {rendererClass=&NiagaraSpriteRendererProperties, rendererIndex=1}
   {rendererClass=&NiagaraSpriteRendererProperties, rendererIndex=2}
  ]
 }
 Squares_and_diamonds {
  emitterUpdateScript [
   &EmitterState(
    Enable Distance Culling=true
    Enable Visibility Culling=true
    MaxDistance=2500
    Module.bApplySpawnCountScaleByDistance=true
    Module.bUseMaxDistance=true
    Module.bUseMinDistance=false
    Scalability Mode=::NewEnumerator1
    Scale Spawn Count=true
    Visibility Cull Response:ENiagaraExecutionStateManagement=::SleepAndClearParticles
   )
   &SpawnRate(SpawnRate=7, Use Spawn Probability=false)
  ]
  particleSpawnScript [
   /SetVariables(Particles.VisibilityTag=&MakeIntFromFloat(Float=&UniformAOrBFloat(A=0, Distribution Weight=0.4))) {isSetParameters=true}
   &InitializeParticle(
    Color=&MakeCustomLinearColorFromBool(
     Bool=&SetBoolByIntComparison(A=$Particles.VisibilityTag, B=0, Comparison Type=::NewEnumerator0)
     True Color=&RandomRangeLinearColor(Minimum=<rgba 0.8148 0.8148 0.8148 0.25>, Maximum=<rgba 0.3327 0.474 0.6654 0.25>)
     False Color=&RandomRangeLinearColor(Minimum=<rgba 0.5405 0.5583 0.4188 1>, Maximum=<rgba 0.6308 0.6308 0.4731 1>)
    )
    Lifetime Max=1
    Lifetime Min=1.5
    Lifetime Mode=::NewEnumerator1
    Mass Mode=::NewEnumerator3
    Mesh Scale Max=<12 12 2>
    Mesh Scale Min=<7 7 1>
    Mesh Scale Mode=::NewEnumerator2
    Sprite Rotation Mode=::NewEnumerator2
    Sprite Size Max=<7 100>
    Sprite Size Min=<3 50>
    Sprite Size Mode=::NewEnumerator2
    UsePositionOffset=false
    Write Material Random=true
   )
   &SphereLocation(
    MaskSpawnGroup=false
    Sphere Distribution:ENiagaraSphereDistributionMode=::NewEnumerator1
    Sphere Radius=31
    U Position:NiagaraFloat=&RandomRangeFloat(Minimum=0, Maximum=1)
    UseNonUniformScale=false
    UseOffset=false
    UseSphereOrientation=false
    V Position:NiagaraFloat=&RandomRangeFloat(Minimum=0.859, Maximum=1)
   )
   &AddVelocity(Velocity=&MakeCustomVectorFromBool(
     Bool=&SetBoolByIntComparison(A=$Particles.VisibilityTag, B=0, Comparison Type=::NewEnumerator0)
     True Vector=&MakeVector(X=0, Y=0, Z=&RandomRangeFloat(Minimum=50, Maximum=100))
     False Vector=<1 0 100>
    ))
   &InitialMeshOrientation(Orientation Axis=<0 0 1>, Orientation Vector=$Output.SphereLocation.SphereNormal, Use Orientation Vector=true, Use Rotation Vector=false)
  ]
  particleUpdateScript [
   &ParticleState()
   &MeshRotationForce(Rotation=&RandomRangeVector(Minimum=<-2 -2 -2>, Maximum=<2 2 2>))
   &ScaleColor(
    Scale Alpha=&Lerp_Float(
     A=&FloatFromCurve(FloatCurve=DataInterface() {propertyValues {Curve {keys [
          {interpMode=RCIM_Cubic, tangentMode=RCTM_User, time=0, value=0}
          {interpMode=RCIM_Cubic, tangentMode=RCTM_User, time=0.1, value=1}
          {tangentMode=RCTM_User, time=1, value=0}
         ]}}})
     B=&FloatFromCurve(FloatCurve=DataInterface() {propertyValues {Curve {keys [
          {interpMode=RCIM_Cubic, tangentMode=RCTM_User, time=0, value=0}
          {interpMode=RCIM_Cubic, tangentMode=RCTM_User, time=0.5, value=1}
          {tangentMode=RCTM_User, time=1, value=0}
         ]}}})
     Alpha=&MakeFloatFromInt(Int=$Particles.Initial.VisibilityTag)
    )
    ScaleA=true
    ScaleRGB=true
   )
   &CurlNoiseForce(
    Noise Frequency=5
    Noise Quality / Cost=::NewEnumerator0
    Noise Strength=&MakeCustomFloatFromBool(
     Bool=&SetBoolByIntComparison(A=$Particles.Initial.VisibilityTag, B=0, Comparison Type=::NewEnumerator0)
     True Float=0
     False Float=30
    )
    Pan Noise=true
    Pan Noise Field=<1 1 1>
   )
   &SolveForcesAndVelocity(Clamp Velocity=false, Limit Acceleration=false)
   &SolveRotationalForcesAndVelocity()
  ]
  rendererClasses [&NiagaraMeshRendererProperties, &NiagaraSpriteRendererProperties]
  renderers [
   {rendererClass=&NiagaraMeshRendererProperties, rendererIndex=0}
   {rendererClass=&NiagaraSpriteRendererProperties, rendererIndex=1}
  ]
 }
}
summary {
 systemName=NS_TimeOrb_Idle
 emitters [
  {emitterName=Fake_Glow, rendererClasses [&NiagaraSpriteRendererProperties]}
  {
   emitterName=Squares_and_diamonds
   rendererClasses [&NiagaraMeshRendererProperties, &NiagaraSpriteRendererProperties]
  }
  {
   emitterName=Icon_And_Light
   rendererClasses [&NiagaraLightRendererProperties, &NiagaraSpriteRendererProperties]
  }
 ]
 userVariables [User.True Vector_Velocity {type=Vector3f, description="", defaultValue {struct=Vector3f, value=<0 0 0>}}]
}
